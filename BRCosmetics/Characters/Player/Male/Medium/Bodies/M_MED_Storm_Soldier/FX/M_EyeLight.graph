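# Unlit, two-sided modulate material for Niagara mesh particles, named as the Storm Soldier's eye light, with a purple default tint.
# Time-panned noise, offset by vertex red, flickers and picks offsets into a lightning sheet, combined with curve-atlas ramps and scrolling noise for the mask.
# That mask, depth fade and view-facing falloff blend from neutral white to the powered color.
# A camera offset plus UV-driven spread displaces the vertices.
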Material /BRCosmetics/Characters/Player/Male/Medium/Bodies/M_MED_Storm_Soldier/FX/M_EyeLight
graph {
 properties {
  bUsedWithNiagaraMeshParticles=true
  blendMode=BLEND_Modulate
  pixelDepthOffsetMode=PDOM_Legacy
  shadingModel=MSM_Unlit
  translucencyPass=MTP_BeforeDOF
  twoSided=true
  usedShadingModels=Unlit
 }
 nodes [
  n1 = ComponentMask(None=Subtract(A=TextureCoordinate() {uTiling=2, vTiling=2})) {r=true}
  n2 = Transform(None=Constant3Vector() {constant=<rgba 1 0 0 0>})
  Add(
   A=MaterialFunctionCall(Offset Amount=ScalarParameter() {defaultValue=10, parameterName=CamOffset}) {materialFunction=/Engine/Functions/Engine_MaterialFunctions02/WorldPositionOffset/CameraOffset}.Clamped Camera Offset
   B=Multiply(A=Multiply(A=@n1, B=@n2), B=ScalarParameter() {defaultValue=2})
  )
  Constant3Vector() {constant=<rgba 0.9688 0.9688 0.9688 0>}
  n3 = MaterialFunctionCall(XY=OneMinus(None=TextureCoordinate())) {materialFunction=&Swizzle}
  n4 = Constant()
  n5 = TextureSample(UVs=Add(A=Panner(Coordinate=@n4) {speedX=0.1, speedY=0.0016}, B=VertexColor().R)) {samplerType=SAMPLERTYPE_LinearColor, texture=/Game/Effects/Fort_Effects/Textures/Tile/LinerNoise64x64}
  n6 = Saturate(None=Add(B=@n5.R) {constA=-0.25})
  n7 = TextureCoordinate() {vTiling=0.75}
  n8 = Constant2Vector() {g=1, r=0.1}
  n9 = Ceil(None=Multiply(A=Constant2Vector() {r=10}, B=@n5.G))
  n10 = Add(
   A=Add(A=Multiply(A=@n7, B=@n8), B=Multiply(A=@n9) {constB=0.1})
   B=ComponentMask(None=VectorParameter() {defaultValue=<rgba 0.022 -0.15 0 0>}.RGB) {g=true, r=true}
  )
  n11 = CurveAtlasRowParameter(CurveTime=ComponentMask(None=TextureCoordinate()) {g=true}) {
   atlas=&AthenaCurveAtlas01
   curve=/Game/Characters/Player/Male/Medium/Bodies/M_MED_Storm_Soldier/FX/Curve_Ramps
   defaultValue=101
   parameterName=Param_1
  }
  n12 = Multiply(
   A=Multiply(
    A=Multiply(A=@n6) {constB=5}
    B=TextureSample(UVs=@n10) {samplerSource=SSM_Clamp_WorldGroupSettings, texture=/Game/Athena/HUD/Materials/T-UI-LightningSheet}.R
   )
   B=@n11.B
  )
  n13 = Multiply(
   A=Add(A=ComponentMask(None=@n3.YX) {r=true}) {constB=0.25}
   B=Multiply(A=VectorParameter() {parameterName=Color, defaultValue=<rgba 3.4716 0.0625 4 1>}.RGB, B=LinearInterpolate(Alpha=@n12) {constA=1, constB=1.5})
  )
  n14 = ScalarParameter() {defaultValue=5, parameterName=Exp}
  n15 = DotProduct(A=VertexNormalWS(), B=CameraVectorWS())
  n16 = MaterialFunctionCall(Float2=TextureCoordinate()) {materialFunction=&BreakOutFloat2Components}
  n17 = CurveAtlasRowParameter(CurveTime=@n16.R) {
   atlas=&AthenaCurveAtlas01
   curve=&Curve_Ramps
   defaultValue=101
   parameterName=Param_1
  }
  n18 = CurveAtlasRowParameter(CurveTime=@n16.G) {
   atlas=&AthenaCurveAtlas01
   curve=&Curve_Ramps
   defaultValue=101
   parameterName=Param_1
  }
  n19 = Panner(Coordinate=TextureCoordinate() {uTiling=0.75, vTiling=0.75}) {speedX=-0.5, speedY=-0.35}
  n20 = Multiply(
   A=Add(A=@n12, B=Multiply(A=@n17.R, B=@n18.G))
   B=Add(A=TextureSample(UVs=@n19) {texture=&T_NoisePack_01}.R) {constB=0.1}
  )
  LinearInterpolate(
   A=Constant() {r=1}
   B=MaterialFunctionCall(Emissive=Power(Base=@n13, Exp=@n14)) {materialFunction=&MF_FX_EmissiveAdjustments}.Result
   Alpha=Multiply(A=Abs(None=@n15), B=DepthFade(Opacity=@n20) {fadeDistanceDefault=20})
  )
 ]
}
