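# Niagara cosmetic body effect for the Storm Soldier character that attaches purple-magenta smoke, electric bolts and glow lights to sockets.
# Smoke sprites, randomized mesh bolts and two light emitters follow a shared system socket reader, spreading across sockets by partition, random or fixed index.
# Array-selected offsets nudge the first socket's particles.
# User.bisHandHidden zeroes spawning in every emitter, plus light radius and sprite and mesh scale, hiding live particles too.
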NiagaraSystem /BRCosmetics/Characters/Player/Male/Medium/Bodies/M_MED_Storm_Soldier/FX/NS_StormSolider_BodyFX
dependencies {
 usedDataInterfaces [
  &NiagaraDataInterfaceSpriteRendererInfo
  &NiagaraDataInterfaceArrayFloat3
  &NiagaraDataInterfaceCurve
  &NiagaraDataInterfaceSocketReader
 ]
 usedDynamicInputs []
 usedModules [
  &SystemState
  &self:SystemSpawnScript.SystemScriptSource.SystemScriptGraph.NiagaraNodeAssignment_0.SetVariables_#1
 ]
}
emitters {
 Lights {
  emitterUpdateScript [
   &EmitterState()
   &SpawnRate(
    SpawnRate=&MakeCustomFloatFromBool(Bool=$User.bisHandHidden, True Float=0, False Float=2)
    Use Spawn Probability=false
   )
  ]
  particleSpawnScript [
   &PartitionParticles()
   &InitializeParticle(Color=<rgba 2.9072 0.2708 4 6>, Lifetime=1.1, UsePositionOffset=false, Write Material Random=true)
  ]
  particleUpdateScript [
   &ParticleState()
   &SocketLocation(
    SamplePosOffset=true
    Sampled Position Offset=&SelectVectorFromArray(
     Direct Array Index=$Particles.PartitionParticles.Partition
     Vector Selection Array=DataInterface() {propertyValues {FloatData [<3 5 0>, <0 0 0>]}}
    )
    Socket Index=$Particles.PartitionParticles.Partition
    Socket Reader=$System.Socket Reader
    Socket Sampling Mode=::NewEnumerator3
    VelocitySpeedLimit=false
   )
   &ScaleColor(
    Scale Alpha=&FloatFromCurve(FloatCurve=DataInterface() {propertyValues {Curve {
        defaultValue=1
        keys [
         {interpMode=RCIM_Cubic, leaveTangentWeight=0.0333, tangentMode=RCTM_User, time=0, value=0}
         {arriveTangentWeight=0.0498, interpMode=RCIM_Cubic, leaveTangentWeight=0.3986, tangentMode=RCTM_User, time=0.1, value=1}
         {arriveTangentWeight=0.3986, interpMode=RCIM_Cubic, leaveTangentWeight=0.0498, tangentMode=RCTM_User, time=0.9, value=1}
         {arriveTangentWeight=0.0333, interpMode=RCIM_Cubic, tangentMode=RCTM_User, time=1, value=0}
        ]
       }}})
    ScaleA=true
    ScaleRGB=true
   )
   &Light_Attributes(
    Falloff Exponent=1
    Light Exponent=4
    Light Radius:NiagaraFloat=&MakeCustomFloatFromBool(Bool=$User.bisHandHidden, True Float=0, False Float=40)
    Radius=1
   )
  ]
  rendererClasses [&NiagaraLightRendererProperties]
  renderers [{rendererClass=&NiagaraLightRendererProperties, rendererIndex=0}]
 }
 Lights001 {
  emitterUpdateScript [
   &EmitterState()
   &SpawnRate(
    SpawnRate=&MakeCustomFloatFromBool(Bool=$User.bisHandHidden, True Float=0, False Float=1)
    Use Spawn Probability=false
   )
  ]
  particleSpawnScript [&InitializeParticle(Color=<rgba 2.9072 0.2708 4 6>, Lifetime=1.1, UsePositionOffset=false, Write Material Random=true)]
  particleUpdateScript [
   &ParticleState()
   &SocketLocation(SamplePosOffset=true, Socket Index=3, Socket Reader=$System.Socket Reader, Socket Sampling Mode=::NewEnumerator3, VelocitySpeedLimit=false)
   &ScaleColor(
    Scale Alpha=&FloatFromCurve(FloatCurve=DataInterface() {propertyValues {Curve {
        defaultValue=1
        keys [
         {interpMode=RCIM_Cubic, leaveTangentWeight=0.0333, tangentMode=RCTM_User, time=0, value=0}
         {arriveTangentWeight=0.0498, interpMode=RCIM_Cubic, leaveTangentWeight=0.3986, tangentMode=RCTM_User, time=0.1, value=1}
         {arriveTangentWeight=0.3986, interpMode=RCIM_Cubic, leaveTangentWeight=0.0498, tangentMode=RCTM_User, time=0.9, value=1}
         {arriveTangentWeight=0.0333, interpMode=RCIM_Cubic, tangentMode=RCTM_User, time=1, value=0}
        ]
       }}})
    ScaleA=true
    ScaleRGB=true
   )
   &Light_Attributes(
    Falloff Exponent=1
    Light Exponent=4
    Light Radius:NiagaraFloat=&MakeCustomFloatFromBool(Bool=$User.bisHandHidden, True Float=0, False Float=40)
    Radius=1
   )
  ]
  rendererClasses [&NiagaraLightRendererProperties]
  renderers [{rendererClass=&NiagaraLightRendererProperties, rendererIndex=0}]
 }
 Long_Bolts {
  emitterUpdateScript [
   &EmitterState(
    Life Cycle Mode=::NewEnumerator1
    Loop Duration=&RandomRangeFloat(Minimum=0, Maximum=0.8)
    Recalculate Duration Each Loop=true
    UseLoopDelay=false
   )
   &SpawnRate(
    SpawnRate=&MakeCustomFloatFromBool(Bool=$User.bisHandHidden, True Float=0, False Float=2)
    Use Spawn Probability=false
   )
   &SpawnBurst_Instantaneous(
    Spawn Count=&MakeCustomIntFromBool(Bool=$User.bisHandHidden, True INT=0, False INT=&RandomRangeInt(Minimum=1, Maximum=3))
    Use Spawn Probability=false
    UseLoopCountLimit=false
   )
  ]
  particleSpawnScript [
   /SetVariables(Particles.Random:NiagaraInt32=&RandomRangeInt(Minimum=0, Maximum=4)) {isSetParameters=true}
   &InitializeParticle(
    Color=<rgba 128 0 60.4464 1>
    Lifetime Max=0.3
    Lifetime Min=0.05
    Lifetime Mode=::NewEnumerator1
    Mesh Scale Max=<0.06 0.25 0.05>
    Mesh Scale Min=<0.03 0.1 0.025>
    Mesh Scale Mode=::NewEnumerator2
    UsePositionOffset=false
    Write Material Random=true
   )
   &ShapeLocation(MaskSpawnGroup=false, Override Seed=false, Sphere Radius=5, UseOwnerScale=true)
   &InitialMeshOrientation(Use Rotation Vector=false)
   &CameraOffset(Camera Offset Amount=10)
   &AddVelocity(Velocity=&Add_Vector(
     A=&Multiply_VectorByFloat(Vector=$Particles.ShapeLocation.ShapeNormal, Float=2)
     B=&RandomRangeVector(Minimum=<0 0 0>, Maximum=<0 35 0>)
    ))
   &DynamicMaterialParameters(
    Index 0 Param 1=&RandomRangeFloat(Minimum=0.9, Maximum=0.95)
    Index 0 Param 2=&RandomRangeFloat(Minimum=0.1, Maximum=0.2)
    Index 0 Param 3=&RandomRangeFloat(Minimum=0.2, Maximum=0.3)
    Module.Index0Float0Write=true
    Module.Index0Float1Write=true
    Module.Index0Float2Write=true
    Module.Index0Float3Write=false
   )
  ]
  particleUpdateScript [
   &ParticleState()
   &SocketLocation(
    SamplePosOffset=true
    Sampled Position Offset=&Add_Vector(
     A=Linked() {linkedVariable {name=Particles.Initial.ShapeLocation.ShapePosition, type=NiagaraPosition}}
     B=&SelectVectorFromArray(
      Direct Array Index=$Particles.Random
      Vector Selection Array=DataInterface() {propertyValues {FloatData [<3 5 0>, <0 0 0>, <0 0 0>, <0 0 0>, <0 0 0>]}}
     )
    )
    Socket Index=$Particles.Random
    Socket Reader=$System.Socket Reader
    Socket Sampling Mode=::NewEnumerator3
    VelocitySpeedLimit=false
   )
   &ScaleMeshSize(Scale Factor=&VectorFromFloat(Value=&FloatFromCurve(
      FloatCurve=DataInterface() {propertyValues {Curve {keys [<at 0 2 RCIM_Cubic>,<at 1 4.89 RCIM_Cubic>]}}}
      Scale Curve=&RandomRangeFloat(Minimum=0.37, Maximum=0.42)
     )))
   &ScaleColor(
    Scale RGB=&VectorFromFloat(Value=&FloatFromCurve(FloatCurve=DataInterface() {propertyValues {Curve {keys [<at 0.4 10>,<at 0.5 0.8103>,<at 1 0.01>]}}}))
    ScaleA=true
    ScaleRGB=true
   )
   &Drag(Drag=5, Use Linear Drag=true, Use Rotational Drag=true)
   &SolveForcesAndVelocity(Clamp Velocity=false, Limit Acceleration=false)
   &ScaleMeshSize(Scale Factor=&VectorFromFloat(Value=&MakeCustomFloatFromBool(Bool=$User.bisHandHidden, True Float=0, False Float=1))) {module=ScaleMeshSize001}
  ]
  rendererClasses [&NiagaraMeshRendererProperties]
  renderers [{rendererClass=&NiagaraMeshRendererProperties, rendererIndex=0}]
 }
 Smokes {
  emitterUpdateScript [
   &EmitterState()
   &SpawnRate(
    SpawnRate=&MakeCustomFloatFromBool(Bool=$User.bisHandHidden, True Float=0, False Float=20)
    Use Spawn Probability=false
   )
  ]
  particleSpawnScript [
   &PartitionParticles(Number Of Partitions=5)
   &InitializeParticle(
    Color=<rgba 0.7661 0.0841 0.85 0.5>
    Lifetime Max=1.5
    Lifetime Mode=::NewEnumerator1
    Sprite Rotation Mode=::NewEnumerator2
    Sprite Size Mode=::NewEnumerator1
    Uniform Sprite Size Max=12
    Uniform Sprite Size Min=10
    UsePositionOffset=false
    Write Material Random=true
   )
   &SubUVAnimation(
    Evaluation Type=::NewEnumerator0
    "Sprite Renderer "=DataInterface() {propertyValues {SpriteRenderer=&self:Smokes_0.NiagaraSpriteRendererProperties_0}}
    SubUV Animation Mode:ENiagara_SubUVLookupModeV2=::NewEnumerator2
    UseEndFrame=false
    UseStartFrame=false
   )
   &ShapeLocation(MaskSpawnGroup=false, Override Seed=false, Sphere Radius=4, UseOwnerScale=true)
   &CameraOffset(Camera Offset Amount=4)
  ]
  particleUpdateScript [
   &ParticleState()
   &SocketLocation(
    SamplePosOffset=true
    Sampled Position Offset=&Add_Vector(
     A=Linked() {linkedVariable {name=Particles.Initial.ShapeLocation.ShapePosition, type=NiagaraPosition}}
     B=&SelectVectorFromArray(
      Direct Array Index=$Particles.PartitionParticles.Partition
      Vector Selection Array=DataInterface() {propertyValues {FloatData [<3 5 0>, <0 0 0>, <0 0 0>, <0 0 0>, <0 0 0>]}}
     )
    )
    Socket Index=$Particles.PartitionParticles.Partition
    Socket Reader=$System.Socket Reader
    Socket Sampling Mode=::NewEnumerator3
    VelocitySpeedLimit=false
   )
   &ScaleSpriteSize(Uniform Curve Sprite Scale=DataInterface() {propertyValues {Curve {keys [
        {arriveTangent=2.799, interpMode=RCIM_Cubic, leaveTangent=2.799, tangentMode=RCTM_User, time=0, value=1}
        {arriveTangent=2.2695, interpMode=RCIM_Cubic, leaveTangent=2.2695, tangentMode=RCTM_User, time=0.2, value=1.5}
        <at 1 2.5 RCIM_Cubic>
       ]}}})
   &ScaleColor(
    Scale Alpha=&FloatFromCurve(FloatCurve=DataInterface() {propertyValues {Curve {keys [
         {arriveTangent=5.4614, interpMode=RCIM_Cubic, leaveTangent=5.4614, tangentMode=RCTM_User, time=0, value=0}
         <at 0.3 1 RCIM_Cubic>
         <at 1 0 RCIM_Cubic>
        ]}}})
    Scale RGB=&VectorFromFloat(Value=&FloatFromCurve(FloatCurve=DataInterface() {propertyValues {Curve {keys [<at 0.4 1>,<at 1 0.2>]}}}))
    ScaleA=true
    ScaleRGB=true
   )
   &ScaleSpriteSize(
    Scale Sprite Size Mode=::NewEnumerator0
    Uniform Scale Factor=&MakeCustomFloatFromBool(Bool=$User.bisHandHidden, True Float=0, False Float=1)
   ) {module=ScaleSpriteSize001}
  ]
  rendererClasses [&NiagaraSpriteRendererProperties]
  renderers [{rendererClass=&NiagaraSpriteRendererProperties}]
 }
}
summary {
 systemName=NS_StormSolider_BodyFX
 emitters [
  {emitterName=Smokes, rendererClasses [&NiagaraSpriteRendererProperties]}
  {emitterName=Long_Bolts, rendererClasses [&NiagaraMeshRendererProperties]}
  {emitterName=Lights, rendererClasses [&NiagaraLightRendererProperties]}
  {emitterName=Lights001, rendererClasses [&NiagaraLightRendererProperties]}
 ]
 userVariables [User.bisHandHidden {type=NiagaraBool}]
}
