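# Niagara head effect for the Storm Soldier character body that shows a mesh-based eyes piece on the head.
# A single Eyes emitter burst-spawns mesh particles at a fixed position offset and does not kill them when their lifetime elapses.
# Each update, the User.bisHeadHidden bool drives mesh scale to zero or one, so the eyes vanish when the head is hidden.
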NiagaraSystem /BRCosmetics/Characters/Player/Male/Medium/Bodies/M_MED_Storm_Soldier/FX/NS_StormSolider_HeadFX
dependencies {usedDataInterfaces [], usedDynamicInputs [], usedModules [&SystemState]}
emitters {Eyes {
  emitterUpdateScript [
   &EmitterState(
    Inactive Response=::NewEnumerator1
    Life Cycle Mode=::NewEnumerator1
    Loop Behavior=::NewEnumerator1
    Loop Duration Mode=::NewEnumerator1
    UseLoopDelay=false
   )
   &SpawnBurst_Instantaneous(Use Spawn Probability=false, UseLoopCountLimit=false)
  ]
  particleSpawnScript [&InitializeParticle(Color Mode=::NewEnumerator0, Mesh Scale Mode=::NewEnumerator3, Position Offset=<13 4 0>, UsePositionOffset=true, Write Material Random=true)]
  particleUpdateScript [
   &ParticleState(Kill Particles When Lifetime Has Elapsed=false, Let Infinitely Lived Particles Die When Emitter Deactivates=true)
   &ScaleMeshSize(Scale Factor=&VectorFromFloat(Value=&MakeCustomFloatFromBool(Bool=$User.bisHeadHidden, True Float=0, False Float=1)))
  ]
  rendererClasses [&NiagaraMeshRendererProperties]
  renderers [{rendererClass=&NiagaraMeshRendererProperties, rendererIndex=0}]
 }}
summary {
 systemName=NS_StormSolider_HeadFX
 emitters [{emitterName=Eyes, rendererClasses [&NiagaraMeshRendererProperties]}]
 userVariables [User.bisHeadHidden {type=NiagaraBool}]
}
