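# Cosmetic animation blueprint for the Storm Soldier body, adding corrective arm deformation and secondary motion to a pose copied from its attached parent mesh.
# Two right-forearm RBF pose drivers for bicep, tricep and elbow bones are blended by bIsSkyDiving.
# The result feeds chained Anim Dynamics nodes for a cable and both leg tubes.
# The EventGraph sets jumpMultVal around jumps, but nothing in the AnimGraph consumes it.
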AnimBlueprint /BRCosmetics/Characters/Player/Male/Medium/Bodies/M_MED_Storm_Soldier/Meshes/M_MED_Storm_Soldier_Skeleton_AnimBP
target_skeleton=&/M_MED_Storm_Soldier_Skeleton
graphs [
 AnimGraph {nodes [
   n1 = Anim Dynamics - Bone: dyn_cable_01(ComponentPose=Local To Component(LocalPose=Use cached pose 'poseDriver_cache'()), Alpha=0.15, ExternalForce="0,0,0") {
    angularDampingOverride=1
    angularSpringConstant=200
    bAngularSpring=true
    bGravityOverrideInSimSpace=true
    bLinearSpring=true
    bOverrideAngularDamping=true
    gravityScale=0.35
    linearDampingOverride=0.5
    linearSpringConstant=150
    simulationSpace=RootRelative
    alphaBoolBlend {blendOption=Linear}
    boundBone {boneName=dyn_cable_01}
    componentAppliedLinearAccClamp=<2000 2000 3000>
    componentLinearAccScale=<1.5 4 1>
    componentLinearVelScale=<0.5 4 1>
    physicsBodyDefinitions [{
      sphereCollisionRadius=0
      boundBone {boneName=dyn_cable_01}
      boxExtents=<15 15 15>
      constraintSetup {angularLimitsMax=<5 10 10>, angularLimitsMin=<-5 -45 -10>, angularTarget=<1 0 0>, linearAxesMax=<2 2 2>, linearAxesMin=<-2 -2 -2>}
      localJointOffset=<8 0 0>
     }]
   }
   n2 = Anim Dynamics - Bone: dyn_legtube_l(ComponentPose=@n1, Alpha=0.5, ExternalForce="0,0,0") {
    angularSpringConstant=500
    bAngularSpring=true
    bGravityOverrideInSimSpace=true
    alphaBoolBlend {blendOption=Linear}
    boundBone {boneName=dyn_legtube_l}
    componentAppliedLinearAccClamp=<1000 1000 1000>
    componentLinearAccScale=<1 1 1>
    componentLinearVelScale=<1 1 1>
    physicsBodyDefinitions [{
      sphereCollisionRadius=0
      boundBone {boneName=dyn_legtube_l}
      boxExtents=<8 8 8>
      constraintSetup {angularLimitsMax=<0 0 15>, angularLimitsMin=<0 0 -15>, angularTarget=<1 0 0>}
      localJointOffset=<0 -0.5 0>
     }]
    relativeSpaceBone {boneName=foot_l}
   }
   n3 = Anim Dynamics - Bone: dyn_legtube_r(ComponentPose=@n2, Alpha=0.5, ExternalForce="0,0,0") {
    angularSpringConstant=500
    bAngularSpring=true
    bGravityOverrideInSimSpace=true
    gravityScale=0.15
    alphaBoolBlend {blendOption=Linear}
    boundBone {boneName=dyn_legtube_r}
    componentAppliedLinearAccClamp=<1000 1000 1000>
    componentLinearAccScale=<0.5 0.5 0.5>
    componentLinearVelScale=<1 1 1>
    physicsBodyDefinitions [{
      sphereCollisionRadius=0
      boundBone {boneName=dyn_legtube_r}
      boxExtents=<8 8 8>
      constraintSetup {angularLimitsMax=<0 0 15>, angularLimitsMin=<0 0 -15>, angularTarget=<1 0 0>}
      localJointOffset=<0 -0.5 0>
     }]
    relativeSpaceBone {boneName=foot_r}
   }
   Output Pose(Result=Component To Local(ComponentPose=@n3)) {layerGroup=None}
   Save cached pose 'CopyPose_cache'(Pose=Copy Pose From Mesh() {bUseAttachedParent=true}) {cacheName=CopyPose_cache}
   n4 = Pose Driver - Source: lowerarm_r!(SourcePose:Pose Link Structure=Use cached pose 'CopyPose_cache'()) -> (Pose:Pose Link Structure) {
    is=title AnimGraphNode_PoseDriver
    poseAsset=&/M_MED_Storm_Soldier_AimGun_PoseAsset
    onlyDriveBones [{boneName=dyn_bicep_01_r}, {boneName=dyn_tricep_01_r}, {boneName=dyn_elbow_01_r}, {boneName=dyn_elbow_02_r}, {boneName=None}]
    poseTargets [
     {
      bApplyCustomCurve=false
      bIsHidden=false
      distanceMethod=DefaultMethod
      drivenName=Pose_6786
      functionType=DefaultFunction
      targetScale=1
      boneTransforms [{targetRotation=<rot -2.196 28.3403 4.0638>, targetTranslation=<-26.6515 0 0>}]
      customCurve {defaultValue=3.4028234663852886e+38, postInfinityExtrap=RCCE_Constant, preInfinityExtrap=RCCE_Constant, keys []}
      targetRotation=<rot 0 0 0>
     }
     {
      bApplyCustomCurve=false
      bIsHidden=false
      distanceMethod=DefaultMethod
      drivenName=Pose_6787
      functionType=DefaultFunction
      targetScale=0.5533
      boneTransforms [{targetRotation=<rot -0.0131 58.4093 0.0501>, targetTranslation=<-26.6515 0 0>}]
      customCurve {defaultValue=3.4028234663852886e+38, postInfinityExtrap=RCCE_Constant, preInfinityExtrap=RCCE_Constant, keys []}
      targetRotation=<rot 0 0 0>
     }
     {
      bApplyCustomCurve=false
      bIsHidden=false
      distanceMethod=DefaultMethod
      drivenName=Pose_6788
      functionType=DefaultFunction
      targetScale=0.4868
      boneTransforms [{targetRotation=<rot -0.0632 39.8909 -0.2134>, targetTranslation=<-26.6515 0 0>}]
      customCurve {defaultValue=3.4028234663852886e+38, postInfinityExtrap=RCCE_Constant, preInfinityExtrap=RCCE_Constant, keys []}
      targetRotation=<rot 0 0 0>
     }
     {
      bApplyCustomCurve=false
      bIsHidden=false
      distanceMethod=DefaultMethod
      drivenName=Pose_6789
      functionType=DefaultFunction
      targetScale=0.7283
      boneTransforms [{targetRotation=<rot 0.0479 108.9856 -0.1841>, targetTranslation=<-26.6515 0 0>}]
      customCurve {defaultValue=3.4028234663852886e+38, postInfinityExtrap=RCCE_Constant, preInfinityExtrap=RCCE_Constant, keys []}
      targetRotation=<rot 0 0 0>
     }
     {
      bApplyCustomCurve=false
      bIsHidden=false
      distanceMethod=DefaultMethod
      drivenName=Pose_6790
      functionType=DefaultFunction
      targetScale=0.3508
      boneTransforms [{targetRotation=<rot -0.5961 77.4734 0.0617>, targetTranslation=<-26.6515 0 0>}]
      customCurve {defaultValue=3.4028234663852886e+38, postInfinityExtrap=RCCE_Constant, preInfinityExtrap=RCCE_Constant, keys []}
      targetRotation=<rot 0 0 0>
     }
     {
      bApplyCustomCurve=false
      bIsHidden=false
      distanceMethod=DefaultMethod
      drivenName=Pose_6791
      functionType=DefaultFunction
      targetScale=0.4586
      boneTransforms [{targetRotation=<rot -0.0889 45.9325 0.0114>, targetTranslation=<-26.6515 0 0>}]
      customCurve {defaultValue=3.4028234663852886e+38, postInfinityExtrap=RCCE_Constant, preInfinityExtrap=RCCE_Constant, keys []}
      targetRotation=<rot 0 0 0>
     }
     {
      bApplyCustomCurve=false
      bIsHidden=false
      distanceMethod=DefaultMethod
      drivenName=Pose_6792
      functionType=DefaultFunction
      targetScale=0.4586
      boneTransforms [{targetRotation=<rot -0.7642 51.5809 0.2552>, targetTranslation=<-26.6515 0 0>}]
      customCurve {defaultValue=3.4028234663852886e+38, postInfinityExtrap=RCCE_Constant, preInfinityExtrap=RCCE_Constant, keys []}
      targetRotation=<rot 0 0 0>
     }
     {
      bApplyCustomCurve=false
      bIsHidden=false
      distanceMethod=DefaultMethod
      drivenName=Pose_6793
      functionType=DefaultFunction
      targetScale=0.5588
      boneTransforms [{targetRotation=<rot -1.7019 65.0964 0.7299>, targetTranslation=<-26.6515 0 0>}]
      customCurve {defaultValue=3.4028234663852886e+38, postInfinityExtrap=RCCE_Constant, preInfinityExtrap=RCCE_Constant, keys []}
      targetRotation=<rot 0 0 0>
     }
     {
      bApplyCustomCurve=false
      bIsHidden=false
      distanceMethod=DefaultMethod
      drivenName=Pose_6794
      functionType=DefaultFunction
      targetScale=0.3508
      boneTransforms [{targetRotation=<rot -2.3928 81.2596 1.1593>, targetTranslation=<-26.6515 0 0>}]
      customCurve {defaultValue=3.4028234663852886e+38, postInfinityExtrap=RCCE_Constant, preInfinityExtrap=RCCE_Constant, keys []}
      targetRotation=<rot 0 0 0>
     }
     {
      bApplyCustomCurve=false
      bIsHidden=false
      distanceMethod=DefaultMethod
      drivenName=Pose_6795
      functionType=DefaultFunction
      targetScale=0.4586
      boneTransforms [{targetRotation=<rot -2.6509 94.8178 1.4715>, targetTranslation=<-26.6515 0 0>}]
      customCurve {defaultValue=3.4028234663852886e+38, postInfinityExtrap=RCCE_Constant, preInfinityExtrap=RCCE_Constant, keys []}
      targetRotation=<rot 0 0 0>
     }
     {
      bApplyCustomCurve=false
      bIsHidden=false
      distanceMethod=DefaultMethod
      drivenName=Pose_6796
      functionType=DefaultFunction
      targetScale=0.4586
      boneTransforms [{targetRotation=<rot -2.6732 100.5054 1.6052>, targetTranslation=<-26.6515 0 0>}]
      customCurve {defaultValue=3.4028234663852886e+38, postInfinityExtrap=RCCE_Constant, preInfinityExtrap=RCCE_Constant, keys []}
      targetRotation=<rot 0 0 0>
     }
    ]
    rBFParams {distanceMethod=Quaternion, radius=20}
    sourceBones [{boneName=lowerarm_r}]
   }
   n5 = Pose Driver - Source: lowerarm_r(SourcePose=Use cached pose 'CopyPose_cache'()) {
    poseAsset=&M_MED_Storm_Soldier_AimGun_PoseAsset
    onlyDriveBones [{boneName=dyn_bicep_01_r}, {boneName=dyn_tricep_01_r}, {boneName=dyn_elbow_01_r}, {boneName=dyn_elbow_02_r}, {boneName=None}]
    poseTargets [
     {
      bApplyCustomCurve=false
      bIsHidden=false
      distanceMethod=DefaultMethod
      drivenName=Pose_6786
      functionType=DefaultFunction
      targetScale=1
      boneTransforms [{targetRotation=<rot -2.196 28.3403 4.0638>, targetTranslation=<-26.6515 0 0>}]
      customCurve {defaultValue=3.4028234663852886e+38, postInfinityExtrap=RCCE_Constant, preInfinityExtrap=RCCE_Constant, keys []}
      targetRotation=<rot 0 0 0>
     }
     {
      bApplyCustomCurve=false
      bIsHidden=false
      distanceMethod=DefaultMethod
      drivenName=Pose_6787
      functionType=DefaultFunction
      targetScale=0.585
      boneTransforms [{targetRotation=<rot -0.0131 58.4093 0.0501>, targetTranslation=<-26.6515 0 0>}]
      customCurve {defaultValue=3.4028234663852886e+38, postInfinityExtrap=RCCE_Constant, preInfinityExtrap=RCCE_Constant, keys []}
      targetRotation=<rot 0 0 0>
     }
     {
      bApplyCustomCurve=false
      bIsHidden=false
      distanceMethod=DefaultMethod
      drivenName=Pose_6788
      functionType=DefaultFunction
      targetScale=0.5145
      boneTransforms [{targetRotation=<rot -0.0632 39.8909 -0.2134>, targetTranslation=<-26.6515 0 0>}]
      customCurve {defaultValue=3.4028234663852886e+38, postInfinityExtrap=RCCE_Constant, preInfinityExtrap=RCCE_Constant, keys []}
      targetRotation=<rot 0 0 0>
     }
     {
      bApplyCustomCurve=false
      bIsHidden=false
      distanceMethod=DefaultMethod
      drivenName=Pose_6789
      functionType=DefaultFunction
      targetScale=0.7582
      boneTransforms [{targetRotation=<rot 0.0479 108.9856 -0.1841>, targetTranslation=<-26.6515 0 0>}]
      customCurve {defaultValue=3.4028234663852886e+38, postInfinityExtrap=RCCE_Constant, preInfinityExtrap=RCCE_Constant, keys []}
      targetRotation=<rot 0 0 0>
     }
     {
      bApplyCustomCurve=false
      bIsHidden=false
      distanceMethod=DefaultMethod
      drivenName=Pose_6790
      functionType=DefaultFunction
      targetScale=0.3568
      boneTransforms [{targetRotation=<rot -0.5961 77.4734 0.0617>, targetTranslation=<-26.6515 0 0>}]
      customCurve {defaultValue=3.4028234663852886e+38, postInfinityExtrap=RCCE_Constant, preInfinityExtrap=RCCE_Constant, keys []}
      targetRotation=<rot 0 0 0>
     }
     {
      bApplyCustomCurve=false
      bIsHidden=false
      distanceMethod=DefaultMethod
      drivenName=Pose_6791
      functionType=DefaultFunction
      targetScale=0.4844
      boneTransforms [{targetRotation=<rot -0.0889 45.9325 0.0114>, targetTranslation=<-26.6515 0 0>}]
      customCurve {defaultValue=3.4028234663852886e+38, postInfinityExtrap=RCCE_Constant, preInfinityExtrap=RCCE_Constant, keys []}
      targetRotation=<rot 0 0 0>
     }
     {
      bApplyCustomCurve=false
      bIsHidden=false
      distanceMethod=DefaultMethod
      drivenName=Pose_6792
      functionType=DefaultFunction
      targetScale=0.4844
      boneTransforms [{targetRotation=<rot -0.7642 51.5809 0.2552>, targetTranslation=<-26.6515 0 0>}]
      customCurve {defaultValue=3.4028234663852886e+38, postInfinityExtrap=RCCE_Constant, preInfinityExtrap=RCCE_Constant, keys []}
      targetRotation=<rot 0 0 0>
     }
     {
      bApplyCustomCurve=false
      bIsHidden=false
      distanceMethod=DefaultMethod
      drivenName=Pose_6793
      functionType=DefaultFunction
      targetScale=0.5872
      boneTransforms [{targetRotation=<rot -1.7019 65.0964 0.7299>, targetTranslation=<-26.6515 0 0>}]
      customCurve {defaultValue=3.4028234663852886e+38, postInfinityExtrap=RCCE_Constant, preInfinityExtrap=RCCE_Constant, keys []}
      targetRotation=<rot 0 0 0>
     }
     {
      bApplyCustomCurve=false
      bIsHidden=false
      distanceMethod=DefaultMethod
      drivenName=Pose_6794
      functionType=DefaultFunction
      targetScale=0.3568
      boneTransforms [{targetRotation=<rot -2.3928 81.2596 1.1593>, targetTranslation=<-26.6515 0 0>}]
      customCurve {defaultValue=3.4028234663852886e+38, postInfinityExtrap=RCCE_Constant, preInfinityExtrap=RCCE_Constant, keys []}
      targetRotation=<rot 0 0 0>
     }
     {
      bApplyCustomCurve=false
      bIsHidden=false
      distanceMethod=DefaultMethod
      drivenName=Pose_6795
      functionType=DefaultFunction
      targetScale=0.4838
      boneTransforms [{targetRotation=<rot -2.6509 94.8178 1.4715>, targetTranslation=<-26.6515 0 0>}]
      customCurve {defaultValue=3.4028234663852886e+38, postInfinityExtrap=RCCE_Constant, preInfinityExtrap=RCCE_Constant, keys []}
      targetRotation=<rot 0 0 0>
     }
     {
      bApplyCustomCurve=false
      bIsHidden=false
      distanceMethod=DefaultMethod
      drivenName=Pose_6796
      functionType=DefaultFunction
      targetScale=0.4838
      boneTransforms [{targetRotation=<rot -2.6732 100.5054 1.6052>, targetTranslation=<-26.6515 0 0>}]
      customCurve {defaultValue=3.4028234663852886e+38, postInfinityExtrap=RCCE_Constant, preInfinityExtrap=RCCE_Constant, keys []}
      targetRotation=<rot 0 0 0>
     }
    ]
    rBFParams {distanceMethod=Quaternion, radius=20}
    sourceBones [{boneName=lowerarm_r}]
   }
   Save cached pose 'poseDriver_cache'(Pose=Blend Poses by bool(BlendPose_0=@n5, BlendPose_1=@n4, bActiveValue=Get bIsSkyDiving()) {blendType=Linear}) {cacheName=poseDriver_cache}
   Get jumpMultVal() -> (jumpMultVal:Vector)
  ]}
 EventGraph {nodes [
   Event BlueprintUpdateAnimation()
   > n1 = Branch(Condition=Get bIsJumping())
   > Delay(Duration=0.2)
   > n2 = Set jumpMultVal(jumpMultVal:Vector="0,0,600") -> (Output_Get:Vector)
   Do Once(execute=[@n2.then, @n1.else])
   > Set jumpMultVal(jumpMultVal="0, 0, 0")
   TryGetPawnOwner()
  ]}
]
variables [jumpMultVal {type=Vector}]
